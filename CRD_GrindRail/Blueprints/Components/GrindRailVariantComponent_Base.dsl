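; Base component for grind rail style variants that applies a style struct's visual and gameplay data to its owning grind rail device.
; Initialization async-loads tip mesh, material and spline mesh in a chain, then refreshes spline visuals and sets the Chrono flag.
; Outside the editor, it also assigns gameplay cues, grinding effect and sound library to the rail.
; Spline tangent changes call UpdateInstancedStaticMeshes, an empty hook like CleanUpVisuals.
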(event EventBeginPlay)

(event EventTick
  (params (DeltaSeconds :type Float%20%28single-precision%29)))

(event Custom|LoadAsyncVisualData
  (params (GrindRailStyle :type FGrind%20Rail%20Style%20Structure))
  (bind (_tipmesh_29_54f2e8be4981e7a3436382a328537f37 _tipmaterial_32_bddfe3ef46dac03321cc7398ba1bef81 _splinemesh_38_ec25dee14ff7bf4620b205bb5de92387 _bischrono_45_13d42e5149bfb26607592884bcb69850) (Utilities|Struct|BreakFGrindRailStyle GrindRailStyle))
  (bind _owninggrindrail (Variables|Default|GetOwningGrindRail))
  (bind _loaded_object (Default|AsyncLoadEditorSafe _tipmesh_29_54f2e8be4981e7a3436382a328537f37)
    (:Loaded
      (bind _asstatic_mesh (Utilities|Casting|CastToStaticMesh _loaded_object)
        (:then
          (Variables|Default|SetLoadedTipMesh _asstatic_mesh)
          (bind _loaded_object_2 (Default|AsyncLoadEditorSafe _tipmaterial_32_bddfe3ef46dac03321cc7398ba1bef81)
            (:Loaded
              (bind _asmaterial_interface (Utilities|Casting|CastToMaterialInterface _loaded_object_2)
                (:then
                  (Variables|Default|SetLoadedTipMaterial _asmaterial_interface))
                (:CastFailed
                  _))
              (bind _loaded_object_3 (Default|AsyncLoadEditorSafe _splinemesh_38_ec25dee14ff7bf4620b205bb5de92387)
                (:Loaded
                  (bind _asstatic_mesh_2 (Utilities|Casting|CastToStaticMesh _loaded_object_3)
                    (:then
                      (Variables|Default|SetLoadedSplineMesh _asstatic_mesh_2))
                    (:CastFailed
                      _))
                  (Spline|UpdateVisualsforSpline _owninggrindrail)
                  (Variables|/GrindRail/Blueprints/BGA_GrindRail.BGA_GrindRail_C|SetIsChronoGrindRail _bischrono_45_13d42e5149bfb26607592884bcb69850 _owninggrindrail))))))
        (:CastFailed
          _)))))

(event Custom|LoadAsyncGameplayData
  (params (GrindRailStyle :type FGrind%20Rail%20Style%20Structure))
  (bind (_gc_sprint_11_70551be647e3e1361b399cacf8675a57 _gc_landonrail_12_a98f0027415073d336d181996155bfc0 _gc_impactplayer_13_dacd0862477eab1274f63d883b2c31a0 _ge_grinding_26_7d28d584419e0bd9298fb88ff86b74c0 _soundlibrary_15_3689008241d02561b01e21b0c1aaf547) (Utilities|Struct|BreakFGrindRailStyle GrindRailStyle))
  (bind _owninggrindrail (Variables|Default|GetOwningGrindRail))
  (if (not (Editor|InEditor))
    (Variables|/GrindRail/Blueprints/BGA_GrindRail.BGA_GrindRail_C|SetGC_Sprinting _gc_sprint_11_70551be647e3e1361b399cacf8675a57 _owninggrindrail)
    (Variables|/GrindRail/Blueprints/BGA_GrindRail.BGA_GrindRail_C|SetGC_LandOnRail _gc_landonrail_12_a98f0027415073d336d181996155bfc0 _owninggrindrail)
    (Variables|/GrindRail/Blueprints/BGA_GrindRail.BGA_GrindRail_C|SetGC_ImpactPlayer _gc_impactplayer_13_dacd0862477eab1274f63d883b2c31a0 _owninggrindrail)
    (bind _loaded_class (Default|AsyncLoadClassEditorSafe _ge_grinding_26_7d28d584419e0bd9298fb88ff86b74c0)
      (:Loaded
        (bind _asgameplay_effect (Utilities|Casting|CastToGameplayEffectClass _loaded_class)
          (:then
            (Variables|/GrindRail/Blueprints/BGA_GrindRail.BGA_GrindRail_C|SetGE_GrindingIncludingInAir _asgameplay_effect (Variables|Default|GetOwningGrindRail)))
          (:CastFailed
            _))
        (bind _loaded_object (Default|AsyncLoadEditorSafe _soundlibrary_15_3689008241d02561b01e21b0c1aaf547)
          (:Loaded
            (bind _assound_library (Utilities|Casting|CastToSoundLibrary _loaded_object))
            (Variables|/GrindRail/Blueprints/BGA_GrindRail.BGA_GrindRail_C|SetGrindRailSoundLibrary _assound_library (Variables|Default|GetOwningGrindRail))))))))

(event Custom|OnSplinePointTangentsChanged
  (params (UpdatedIndex :type Integer))
  (Default|UpdateInstancedStaticMeshes))

(fn ApplyVariantVisuals
  (params (VisualStyle :type FGrind%20Rail%20Style%20Structure))
  (CallFunction|LoadAsyncVisualData VisualStyle)
  (CallFunction|LoadAsyncGameplayData VisualStyle))

(fn CleanUpVisuals)

(fn GetSplineMaterials
  (return (Variables|Default|GetSplineMaterials)))

(fn GetSceneComponents
  (return 0))

(fn GetSplineMesh
  (params (OptUseSegmentedMesh :type Boolean) (OptSplineSegmentLength :type Float%20%28double-precision%29))
  (return (Variables|Default|GetLoadedSplineMesh)))

(fn GetTipMesh
  (return (Variables|Default|GetLoadedTipMesh)))

(fn InitializeVariantComponent
  (params (VisualStyleData :type FGrind%20Rail%20Style%20Structure))
  (bind _owninggrindrail (Variables|Default|GetOwningGrindRail))
  (bind _returnvalue (Actor|GetComponentbyClass _owninggrindrail "/Script/FortniteGame.FortCreativeSplineModificationComponent"))
  (bind _asdevice_grind_rail (Utilities|Casting|CastToDevice_GrindRail (Components|GetOwner)))
  (Variables|Default|SetOwningGrindRail _asdevice_grind_rail)
  (CallFunction|ApplyVariantVisuals VisualStyleData)
  (Utilities|IsValid _returnvalue
    (:Is%20Valid
      (EventDispatchers|BindEventtoOnSplinePointTangentsUpdated _returnvalue (EventDispatchers|CreateEvent :!event OnSplinePointTangentsChanged)))))

(fn SetDefaultMaterials
  (bind _owninggrindrail (Variables|Default|GetOwningGrindRail))
  (Utilities|IsValid _owninggrindrail
    (:Is%20Valid
      (Rendering|Material|SetMaterial (Variables|/GrindRail/Blueprints/BGA_GrindRail.BGA_GrindRail_C|GetEndTip _owninggrindrail) 0 (Variables|Default|GetLoadedTipMaterial)))))

(fn UpdateInstancedStaticMeshes)

(fn UpdateTipMesh
  (params (TipComponent :type Static%20Mesh%20Component%20Object%20Reference))
  (bind _tipmesh (Default|GetTipMesh))
  (if (or (not (Utilities|IsValid (Variables|/Script/Engine.StaticMeshComponent|GetStaticMesh TipComponent) :!class /Script/Engine.KismetSystemLibrary)) (!= (Variables|/Script/Engine.StaticMeshComponent|GetStaticMesh TipComponent) _tipmesh))
    (Components|StaticMesh|SetStaticMesh TipComponent _tipmesh)))
