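; Neon City variant component for the grind rail, choosing which mesh the rail's spline uses.
; GetSplineMesh returns the loaded spline mesh when the segmented-mesh option is set, and otherwise the default rail mesh.
; The segment length input is unused, and the BeginPlay and Tick events are disabled stubs.
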(event EventBeginPlay
  (|CallParentFunction) ; This node is disabled and will not be called.\nDrag off pins to build functionality.
)

(event EventTick
  (params (DeltaSeconds :type Float%20%28single-precision%29))
  (|CallParentFunction DeltaSeconds) ; This node is disabled and will not be called.\nDrag off pins to build functionality.
)

(fn GetSplineMesh
  (params (OptUseSegmentedMesh :type Boolean) (OptSplineSegmentLength :type Float%20%28double-precision%29))
  (return (select OptUseSegmentedMesh (Variables|Default|GetLoadedSplineMesh) (Variables|Default|GetMesh_Rail_Default))))
