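; Vine visual variant component for the Creative grind rail device, dressing the rail spline with vine meshes and moss.
; Spline meshes are picked by segment length, and tip meshes come from the device's tip type.
; A transient HISM scatters evenly spaced moss along the spline with random location and rotation jitter, without collision or shadows.
; Moss is skipped on mobile, respects the device's moss and visibility settings, and ApplyVariantVisuals spawns it only without authority.
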(event EventBeginPlay
  (|CallParentFunction))

(event EventTick
  (params (DeltaSeconds :type Float%20%28single-precision%29))
  (|CallParentFunction DeltaSeconds) ; This node is disabled and will not be called.\nDrag off pins to build functionality.
)

(event EventLoadAsyncVisualData
  (params (GrindRailStyle :type FGrind%20Rail%20Style%20Structure))
  (Spline|UpdateVisualsforSpline (Variables|Default|GetOwningGrindRail)))

(fn AddMossInstances
  (bind _returnvalue 5)
  (bind _returnvalue_2 512.0)
  (bind _moss_hism (Variables|Default|GetMoss_HISM))
  (Utilities|IsValid _moss_hism
    (:Is%20Valid
      (Components|InstancedStaticMesh|ClearInstances _moss_hism)
      (Collision|SetCollisionEnabled _moss_hism "NoCollision")
      (LOD|SetMaxDrawDistance (Variables|Default|GetMoss_HISM) 6e+03)
      (for _index (range 3 (+ (Math|Float|Truncate (* (/ (Spline|GetSplineLength (Variables|/GrindRail/Blueprints/BGA_GrindRail.BGA_GrindRail_C|GetSpline (Variables|Default|GetOwningGrindRail))) _returnvalue_2) _returnvalue)) 1))
        (Variables|Default|SetSpaceBetweenMoss (/ _returnvalue_2 _returnvalue))
        (bind _transform (CallFunction|CreateMossTransform _index))
        (Components|InstancedStaticMesh|AddInstance (Variables|Default|GetMoss_HISM) _transform false)))))

(fn ApplyVariantVisuals
  (params (VisualStyle :type FGrind%20Rail%20Style%20Structure))
  (|CallParentFunction VisualStyle)
  (Networking|SwitchHasAuthority
    (:Remote
      (CallFunction|SpawnMossHISMs))))

(fn CleanUpVisuals
  (bind _moss_hism (Variables|Default|GetMoss_HISM))
  (|CallParentFunction)
  (Utilities|IsValid _moss_hism
    (:Is%20Valid
      (Components|InstancedStaticMesh|ClearInstances _moss_hism))))

(fn CreateMossTransform
  (params (MossIndex :type Integer))
  (return (Math|Transform|MakeTransform (- (Spline|GetLocationatDistanceAlongSpline (Variables|/GrindRail/Blueprints/BGA_GrindRail.BGA_GrindRail_C|GetSpline (Variables|Default|GetOwningGrindRail)) (+ (* MossIndex (Variables|Default|GetSpaceBetweenMoss)) (Utilities|GetRandomValueforMoss)) "Local") (Math|Vector|MakeVector (Utilities|GetRandomValueforMoss) (Utilities|GetRandomValueforMoss) (+ (Math|Float|Absolute(Float) (Utilities|GetRandomValueforMoss)) (LocalVariables|GetMossRandomLocationVerticalOffset)))) (Math|Rotator|CombineRotators (Spline|GetRotationatDistanceAlongSpline (Variables|/GrindRail/Blueprints/BGA_GrindRail.BGA_GrindRail_C|GetSpline (Variables|Default|GetOwningGrindRail)) (+ (* MossIndex (Variables|Default|GetSpaceBetweenMoss)) (Utilities|GetRandomValueforMoss)) "Local") (Math|Vector|RotatorfromAxisandAngle (Spline|GetTangentatDistanceAlongSpline (Variables|/GrindRail/Blueprints/BGA_GrindRail.BGA_GrindRail_C|GetSpline (Variables|Default|GetOwningGrindRail)) (+ (* MossIndex (Variables|Default|GetSpaceBetweenMoss)) (Utilities|GetRandomValueforMoss)) "Local") (* (Utilities|GetRandomValueforMoss) (LocalVariables|GetMossRandomRotationOffset)))) (Variables|Default|GetMossScale))))

(fn GetSceneComponents
  (return (Utilities|Array|MakeArray (Variables|Default|GetMoss_HISM))))

(fn GetSplineMesh
  (params (OptUseSegmentedMesh :type Boolean) (OptSplineSegmentLength :type Float%20%28double-precision%29))
  (bind _array_element (Utilities|Array|ForEachLoopwithBreak (Variables|Default|GetLengthToMeshData))
    (:LoopBody
      (bind (_array_element_mesh_5_c75f5aec4289371b608ab796fe7dbab7 _array_element_maxlengthtousefor_2_1b54df844d592ffc6c68bd99036c99d7) (Utilities|Struct|BreakFortVineRailMeshData _array_element))
      (if (<= OptSplineSegmentLength _array_element_maxlengthtousefor_2_1b54df844d592ffc6c68bd99036c99d7)
        (LocalVariables|SetFoundMesh _array_element_mesh_5_c75f5aec4289371b608ab796fe7dbab7)))
    (:Completed
      _))
  (Utilities|IsValid (LocalVariables|GetFoundMesh)
    (:Is%20Valid
      _)
    (:Is%20Not%20Valid
      (bind _output_get (LocalVariables|SetFoundMesh (Utilities|Struct|BreakFortVineRailMeshData (Utilities|Array|Get(acopy) (Variables|Default|GetLengthToMeshData) (Utilities|Array|LastIndex (Variables|Default|GetLengthToMeshData))))))))
  (return _output_get))

(fn GetTipMesh
  (bind _owninggrindrail (Variables|Default|GetOwningGrindRail))
  (bind _vinetiptype (Variables|/CRD_GrindRail/Device_GrindRail.Device_GrindRail_C|GetVineTipType _owninggrindrail))
  (bind _tiptypetomesh (Variables|Default|GetTipTypeToMesh))
  (bind (_value _returnvalue) (Utilities|Map|Find _tiptypetomesh _vinetiptype))
  (Utilities|IsValid _owninggrindrail
    (:Is%20Valid
      (if _returnvalue
        (return _value)
        (else
          _)))
    (:Is%20Not%20Valid
      _))
  (return (|CallParentFunction)))

(fn SetDefaultMaterials
  (bind _owninggrindrail (Variables|Default|GetOwningGrindRail))
  (Utilities|IsValid _owninggrindrail
    (:Is%20Valid
      (Rendering|Material|SetMaterial (Variables|/GrindRail/Blueprints/BGA_GrindRail.BGA_GrindRail_C|GetEndTip (Variables|Default|GetOwningGrindRail)) 0 (StaticMesh|GetMaterial (Utilities|Map|Find (Variables|Default|GetTipTypeToMesh) (Variables|/CRD_GrindRail/Device_GrindRail.Device_GrindRail_C|GetVineTipType _owninggrindrail)) 0))
      (Rendering|Material|SetMaterial (Variables|Default|GetMoss_HISM) 0 (Variables|Default|GetMossMaterial)))))

(fn SetUpHISM
  (bind _moss_hism (Variables|Default|GetMoss_HISM))
  (if (not (Utilities|IsValid (Variables|Default|GetMoss_HISM) :!class /Script/Engine.KismetSystemLibrary))
    (bind _spawnedcomponent (Fortnite|Utilities|SpawnTransientSceneComponent "/Script/Engine.HierarchicalInstancedStaticMeshComponent" (Variables|Default|GetOwningGrindRail) (Variables|/Script/Engine.Actor|GetRootComponent (Variables|Default|GetOwningGrindRail))))
    (bind _output_get (Variables|Default|SetMoss_HISM _spawnedcomponent))
    (else
      _))
  (Components|StaticMesh|SetStaticMesh _output_get (Variables|Default|GetMossMesh))
  (Rendering|Material|SetMaterial _moss_hism 0 (Variables|Default|GetMossMaterial))
  (Rendering|SetCastShadow _moss_hism false)
  (Components|InstancedStaticMesh|SetCullDistances _moss_hism 0 3000))

(fn SpawnMossHISMs
  (if (not (TouchPlatform|IsMobileGame))
    (if (and (Variables|/CRD_GrindRail/Device_GrindRail.Device_GrindRail_C|GetbApplyVineMoss (Variables|Default|GetOwningGrindRail)) (Visibility|ShouldbeVisible (Variables|/CRD_GrindRail/Device_GrindRail.Device_GrindRail_C|GetCreative_VisibleInGame_Component (Variables|Default|GetOwningGrindRail))))
      (CallFunction|AddMossInstances)
      (else
        (Components|InstancedStaticMesh|ClearInstances (Variables|Default|GetMoss_HISM))))))

(fn UpdateInstancedStaticMeshes
  (bind _moss_hism (Variables|Default|GetMoss_HISM))
  (CallFunction|SetUpHISM)
  (Collision|SetCollisionEnabled _moss_hism "NoCollision")
  (Components|InstancedStaticMesh|ClearInstances _moss_hism)
  (CallFunction|SpawnMossHISMs))
