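; Wire variant component for the Device_GrindRail device that hangs decoration meshes, such as lights, along the rail's spline.
; It spawns a transient, collision-free HISM on the owning rail and clears it on cleanup or update.
; The rail's Wire Decoration style keys maps of meshes and materials to async-load, with an invisible preview material substituted when flagged hidden.
; Once the mesh loads, instances are added at evenly spaced, vertically offset spline points with a max draw distance.
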(event EventBeginPlay
  (|CallParentFunction) ; This node is disabled and will not be called.\nDrag off pins to build functionality.
)

(event EventTick
  (params (DeltaSeconds :type Float%20%28single-precision%29))
  (|CallParentFunction DeltaSeconds) ; This node is disabled and will not be called.\nDrag off pins to build functionality.
)

(event Custom|LoadAndSetLights
  (bind _owninggrindrail (Variables|Default|GetOwningGrindRail))
  (bind _wire_decoration (Variables|/CRD_GrindRail/Device_GrindRail.Device_GrindRail_C|GetWire%20Decoration _owninggrindrail))
  (bind _decorationsmeshesperstyle (Variables|Default|GetDecorationsMeshesPerStyle))
  (bind (_value _returnvalue) (Utilities|Map|Find _decorationsmeshesperstyle _wire_decoration))
  (if _returnvalue
    (bind _loaded_object (Default|AsyncLoadEditorSafe _value)
      (:Loaded
        (bind _asstatic_mesh (Utilities|Casting|CastToStaticMesh _loaded_object))
        (Components|StaticMesh|SetStaticMesh (Variables|Default|GetHangingLights_HISM) _asstatic_mesh)
        (CallFunction|SetLightsMaterial)
        (CallFunction|UpdateLightsHISM)))))

(event Custom|SetLightsMaterial
  (bind _owninggrindrail (Variables|Default|GetOwningGrindRail))
  (bind _wire_decoration (Variables|/CRD_GrindRail/Device_GrindRail.Device_GrindRail_C|GetWire%20Decoration _owninggrindrail))
  (bind _decorationsmaterialsperstyle (Variables|Default|GetDecorationsMaterialsPerStyle))
  (bind (_value _returnvalue) (Utilities|Map|Find _decorationsmaterialsperstyle _wire_decoration))
  (if _returnvalue
    (bind _loaded_object (Default|AsyncLoadEditorSafe _value)
      (:Loaded
        (bind _asmaterial_interface (Utilities|Casting|CastToMaterialInterface _loaded_object))
        (Rendering|Material|SetMaterial (Variables|Default|GetHangingLights_HISM) 0 (select (Class|DeviceGrindRail|GetShouldUseHiddenMaterial (Variables|Default|GetOwningGrindRail)) (Variables|Default|GetInvisiblePreviewMaterial) _asmaterial_interface))))))

(fn CleanUpVisuals
  (bind _hanginglights_hism (Variables|Default|GetHangingLights_HISM))
  (|CallParentFunction)
  (Utilities|IsValid _hanginglights_hism
    (:Is%20Valid
      (Components|InstancedStaticMesh|ClearInstances _hanginglights_hism))))

(fn GetSceneComponents
  (return (Utilities|Array|MakeArray (Variables|Default|GetHangingLights_HISM))))

(fn InitializeVariantComponent
  (params (VisualStyleData :type FGrind%20Rail%20Style%20Structure))
  (|CallParentFunction VisualStyleData)
  (CallFunction|SetupLightsHISM))

(fn SetDefaultMaterials
  (|CallParentFunction)
  (CallFunction|SetLightsMaterial))

(fn SetupLightsHISM
  (if (not (Utilities|IsValid (Variables|Default|GetHangingLights_HISM) :!class /Script/Engine.KismetSystemLibrary))
    (bind _spawnedcomponent (Fortnite|Utilities|SpawnTransientSceneComponent "/Script/Engine.HierarchicalInstancedStaticMeshComponent" (Variables|Default|GetOwningGrindRail) (Variables|/Script/Engine.Actor|GetRootComponent (Variables|Default|GetOwningGrindRail))))
    (Variables|Default|SetHangingLights_HISM _spawnedcomponent)))

(fn UpdateInstancedStaticMeshes
  (bind _hanginglights_hism (Variables|Default|GetHangingLights_HISM))
  (CallFunction|SetupLightsHISM)
  (Collision|SetCollisionEnabled _hanginglights_hism "NoCollision")
  (Components|InstancedStaticMesh|ClearInstances _hanginglights_hism)
  (CallFunction|LoadandSetLights))

(fn UpdateLightsHISM
  (bind _lightdistance (LocalVariables|GetLightDistance))
  (for _index (range (Math|Float|Floor (Math|Float|SafeDivide (Spline|GetSplineLength (Variables|/GrindRail/Blueprints/BGA_GrindRail.BGA_GrindRail_C|GetSpline (Variables|Default|GetOwningGrindRail))) _lightdistance)))
    (Utilities|Array|Add (LocalVariables|GetLightTransforms) (Math|Transform|MakeTransform (+ (Spline|GetLocationatDistanceAlongSpline (Variables|/GrindRail/Blueprints/BGA_GrindRail.BGA_GrindRail_C|GetSpline (Variables|Default|GetOwningGrindRail)) (+ (* _index _lightdistance) (* _lightdistance 0.5)) "Local") (Math|Vector|MakeVector 0.0 0.0 (LocalVariables|GetMeshVerticalOffset)))))
    (LOD|SetMaxDrawDistance (Variables|Default|GetHangingLights_HISM) (LocalVariables|GetMeshCullDistance)))
  (bind _returnvalue (Components|InstancedStaticMesh|AddInstances (Variables|Default|GetHangingLights_HISM) (LocalVariables|GetLightTransforms) false false true))
  (LocalVariables|SetLightInstances_Old _returnvalue))
